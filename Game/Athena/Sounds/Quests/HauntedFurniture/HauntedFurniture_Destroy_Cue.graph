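# SoundCue for the destroy sound of the haunted furniture quest, routed to the Fort_Impacts sound class at reduced volume.
# A single WavePlayer root plays the Quest_HauntedFurniture_Destroy wave, with no randomization or mixing.
# Override attenuation sets a short-range falloff, stereo spread and focus interpolation settings.
# Override concurrency allows one instance at a time, stopping the oldest when a new one starts.
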SoundCue /Game/Athena/Sounds/Quests/HauntedFurniture/HauntedFurniture_Destroy_Cue
root=WavePlayer_0
nodes [{id=WavePlayer_0, properties {soundWaveAssetPtr=&/Quest_HauntedFurniture_Destroy}}]
properties {
 bOverrideAttenuation=true
 bOverrideConcurrency=true
 maxDistance=1024
 soundClassObject=&Fort_Impacts
 volumeMultiplier=0.64
 attenuationOverrides {
  bEnableFocusInterpolation=true
  falloffDistance=512
  focusAttackInterpSpeed=0.5
  focusReleaseInterpSpeed=0.5
  nonFocusVolumeAttenuation=0.2
  stereoSpread=256
  attenuationShapeExtents {x=512}
 }
 concurrencyOverrides {maxCount=1, resolutionRule=StopOldest}
}
