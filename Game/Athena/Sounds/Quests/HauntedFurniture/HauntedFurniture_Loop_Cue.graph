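# Looping ambient sound cue for the haunted furniture quest in Athena.
# A looping wave player feeds an indefinitely looping Enveloper, whose pitch curve rises and falls over a 30-second cycle while volume stays flat.
# Overridden attenuation adds short-range falloff with interpolated listener focus that lowers volume when out of focus.
# Overridden concurrency allows one instance, stopping the quietest.
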SoundCue /Game/Athena/Sounds/Quests/HauntedFurniture/HauntedFurniture_Loop_Cue
root=Enveloper_0
nodes [
 {
  id=Enveloper_0
  children [WavePlayer_2]
  properties {
   bLoop=true
   bLoopIndefinitely=true
   loopEnd=30
   pitchCurve {editorCurveData {keys [{time=0}, <at 15 0.8>, {time=30}]}}
   volumeCurve {editorCurveData {keys [<at 0 1>]}}
  }
 }
 {id=WavePlayer_2, properties {bLooping=true, soundWaveAssetPtr=&/Quest_HauntedFurniture_Loop_01}}
]
properties {
 bOverrideAttenuation=true
 bOverrideConcurrency=true
 maxDistance=1024
 soundClassObject=&Fort_Ambient_Athena
 volumeMultiplier=1.351
 attenuationOverrides {
  bEnableFocusInterpolation=true
  bEnableListenerFocus=true
  falloffDistance=512
  focusAttackInterpSpeed=0.5
  focusReleaseInterpSpeed=0.5
  nonFocusVolumeAttenuation=0.2
  stereoSpread=1000
  attenuationShapeExtents {x=512}
 }
 concurrencyOverrides {maxCount=1, resolutionRule=StopQuietest}
}
